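# Sound cue for opening a wooden barrel loot container.
# It picks one of three equally weighted search-loop-end waves, then applies slight random pitch variation and a short fixed delay before playback.
# It uses item-interaction attenuation and the loot reward sound class, with a concurrency override that prevents new instances once the limit is reached.
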SoundCue /Game/Sounds/Foley_Loot/Containers/WoodBarrel/WoodBarrel_Open_Cue
root=Delay_1
nodes [
 {id=Delay_1, children [Modulator_2], properties {delayMax=0.15, delayMin=0.15}}
 {id=Modulator_2, children [Random_0], properties {pitchMax=1.1, volumeMax=1, volumeMin=1}}
 {id=Random_0, children [WavePlayer_2,WavePlayer_3,WavePlayer_4], properties {bShouldExcludeFromBranchCulling=true, weights [1,1,1]}}
 {id=WavePlayer_2, properties {soundWaveAssetPtr=&/WoodBarrel_Search_Loop_End_01}}
 {id=WavePlayer_3, properties {soundWaveAssetPtr=&/WoodBarrel_Search_Loop_End_02}}
 {id=WavePlayer_4, properties {soundWaveAssetPtr=&/WoodBarrel_Search_Loop_End_03}}
]
properties {
 attenuationSettings=/Game/Sounds/Attenuation/Fort_Item_Interact_Att
 bOverrideConcurrency=true
 maxDistance=1850
 soundClassObject=&Fort_Loot_Reward_PSM
 volumeMultiplier=1.25
 concurrencyOverrides {resolutionRule=PreventNew}
}
