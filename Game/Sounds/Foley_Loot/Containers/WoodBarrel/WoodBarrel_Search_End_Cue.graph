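# Sound cue that plays the ending foley when a search of a wooden barrel loot container finishes.
# A random node picks one of three equally weighted loop-end waves, then a modulator varies its pitch while volume stays fixed.
# The cue plays pitched up and quieter, using prop search attenuation and the loot collect sound class.
# A concurrency override caps simultaneous instances, stopping the farthest first.
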SoundCue /Game/Sounds/Foley_Loot/Containers/WoodBarrel/WoodBarrel_Search_End_Cue
root=Modulator_0
nodes [
 {id=Modulator_0, children [Random_0], properties {pitchMax=1.1, volumeMax=1, volumeMin=1}}
 {id=Random_0, children [WavePlayer_0,WavePlayer_1,WavePlayer_2], properties {bShouldExcludeFromBranchCulling=true, weights [1,1,1]}}
 {id=WavePlayer_0, properties {soundWaveAssetPtr=&/WoodBarrel_Search_Loop_End_01}}
 {id=WavePlayer_1, properties {soundWaveAssetPtr=&/WoodBarrel_Search_Loop_End_02}}
 {id=WavePlayer_2, properties {soundWaveAssetPtr=&/WoodBarrel_Search_Loop_End_03}}
]
properties {
 attenuationSettings=/Game/Sounds/Attenuation/Fort_Prop_Search_Att
 bOverrideConcurrency=true
 maxDistance=1300
 pitchMultiplier=1.5
 soundClassObject=&Fort_Loot_Collect
 volumeMultiplier=0.7
 concurrencyOverrides {maxCount=4, resolutionRule=StopFarthestThenPreventNew}
}
