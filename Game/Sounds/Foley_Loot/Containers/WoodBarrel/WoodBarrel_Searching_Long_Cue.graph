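# Sound cue for the long searching foley of a wooden barrel loot container, attenuated as a searchable prop.
# A mixer layers a quieter, randomly chosen start one-shot, pitched well up, over one of three randomly chosen search loops with slight pitch variation.
# An enveloper fades the loop in over the first half second so the one-shot covers the onset.
# A concurrency override prevents new instances once the limit is reached.
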SoundCue /Game/Sounds/Foley_Loot/Containers/WoodBarrel/WoodBarrel_Searching_Long_Cue
root=Mixer_0
nodes [
 {
  id=Enveloper_0
  children [Modulator_4]
  properties {
   pitchCurve {editorCurveData {keys [{time=0}]}}
   volumeCurve {editorCurveData {keys [<at 0 0 RCIM_Cubic>, {arriveTangent=5.4074, interpMode=RCIM_Cubic, leaveTangent=5.4074, tangentMode=RCTM_User, time=0.5, value=1}]}}
  }
 }
 {id=Mixer_0, children [Modulator_5,Enveloper_0], properties {inputVolume [0.6,1]}}
 {id=Modulator_4, children [Random_2], properties {pitchMax=1.1, volumeMax=1, volumeMin=1}}
 {id=Modulator_5, children [Random_3], properties {pitchMax=1.75, pitchMin=1.5, volumeMax=1, volumeMin=1}}
 {id=Random_2, children [WavePlayer_8, WavePlayer_9, WavePlayer_10], properties {bShouldExcludeFromBranchCulling=true, weights [1,1,1]}}
 {id=Random_3, children [WavePlayer_11, WavePlayer_12, WavePlayer_13], properties {bShouldExcludeFromBranchCulling=true, weights [1,1,1]}}
 {id=WavePlayer_10, properties {bLooping=true, soundWaveAssetPtr=&/WoodBarrel_Search_Loop_03}}
 {id=WavePlayer_11, properties {soundWaveAssetPtr=&/WoodBarrel_Search_Loop_Start_01}}
 {id=WavePlayer_12, properties {soundWaveAssetPtr=&/WoodBarrel_Search_Loop_Start_02}}
 {id=WavePlayer_13, properties {soundWaveAssetPtr=&/WoodBarrel_Search_Loop_Start_03}}
 {id=WavePlayer_8, properties {bLooping=true, soundWaveAssetPtr=&/WoodBarrel_Search_Loop_01}}
 {id=WavePlayer_9, properties {bLooping=true, soundWaveAssetPtr=&/WoodBarrel_Search_Loop_02}}
]
properties {
 attenuationSettings=/Game/Sounds/Attenuation/Fort_Prop_Search_Att
 bOverrideConcurrency=true
 maxDistance=1300
 soundClassObject=&Fort_Loot_Collect
 volumeMultiplier=1.13
 concurrencyOverrides {resolutionRule=PreventNew}
}
